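# Physics asset for the RobotRed two-handed pickaxe cosmetic, supplying collision bodies and joints for its dangling parts.
# Capsules cover the root, handle and chain bones, while dyn_base combines a capsule and box and anchors a cable chain and two cloth chains, four bones each.
# Root, handle and base joints rotate freely; cloth joints swing freely with narrow twist limits; cable joints have tight, asymmetric swing limits and locked twist.
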PhysicsAsset /BRCosmetics/Weapons/FORT_Melee/Pickaxe_RobotRed/Meshes/Pickaxe_RobotRed_Physics
bodies {
 Melee_TwoHanded_Handle [{shapeType=Capsule, length=46, radius=3.6014, center=<1.48 -1.399 46.116>, rotation=<rot -0.562 -90 0.167>}]
 b_AR_Root [{shapeType=Capsule, length=0.505, radius=0.505, rotation=<rot 0 -90 0>}]
 dyn_base [
  {shapeType=Capsule, length=18.1875, radius=8.3342, center=<3.846 4.3518 -11.5462>, rotation=<rot 0.2725 177.3124 -87.2012>}
  {shapeType=Box, extentX=8.71, extentY=19.26, extentZ=15.198, center=<0.7937 -0.171 0>}
 ]
 dyn_cable_01 [{shapeType=Capsule, length=0.6762, radius=1.1132, center=<1.3 -0.1235 0.0869>, rotation=<rot -85.7685 -0.0005 0.0006>}]
 dyn_cable_02 [{shapeType=Capsule, length=1.1762, radius=1.2568, center=<1.9088 0.0001 0.1412>, rotation=<rot -85.7687 0.0002 -0.0001>}]
 dyn_cable_03 [{shapeType=Capsule, length=1.4091, radius=0.5926, center=<1.2158 0 0.0207>, rotation=<rot -89.0226 -0.002 0.002>}]
 dyn_cable_04 [{shapeType=Capsule, length=3.0479, radius=0.5, center=<2.4326 -0.058 0.2253>, rotation=<rot -85.7687 0 -0.0001>}]
 dyn_clothA_01 [{shapeType=Capsule, length=2.6958, radius=1.711, center=<3.2752 0.0914 0.0574>, rotation=<rot 0 -88.4019 -91.0045>}]
 dyn_clothA_02 [{shapeType=Capsule, length=2.304, radius=1.7, center=<2.734 0 0>, rotation=<rot 90 0 0>}]
 dyn_clothA_03 [{shapeType=Capsule, length=1.6769, radius=2.26, center=<3.5067 0 0>, rotation=<rot 90 0 0>}]
 dyn_clothA_04 [{shapeType=Capsule, length=5.3264, radius=1.72, center=<2.7205 0 0.6018>, rotation=<rot 51.161 0 0>}]
 dyn_clothB_01 [{shapeType=Capsule, length=0.944, radius=1.4081, center=<2.2863 0.0103 0.0261>, rotation=<rot 21.5382 -89.9996 89.2939>}]
 dyn_clothB_02 [{shapeType=Capsule, length=1.7023, radius=1.0784, center=<3.1561 0 0.0493>, rotation=<rot 0 -89.9995 -90.8965>}]
 dyn_clothB_03 [{shapeType=Capsule, length=3.2934, radius=2.1787, center=<3.1756 -0.7624 0.2475>}]
 dyn_clothB_04 [{shapeType=Capsule, length=2.6592, radius=2.0813, center=<2.6845 -0.0838 0.5528>, rotation=<rot -4.0969 -90 17.9237>}]
}
constraints [
 {
  bone1Name=dyn_clothA_03
  bone2Name=dyn_clothA_02
  swing1LimitDegrees=45
  swing1Motion=Free
  swing2LimitDegrees=45
  swing2Motion=Free
  twistLimitDegrees=10
 }
 {
  bone1Name=dyn_clothA_04
  bone2Name=dyn_clothA_03
  swing1LimitDegrees=45
  swing1Motion=Free
  swing2LimitDegrees=45
  swing2Motion=Free
  twistLimitDegrees=10
 }
 {
  bone1Name=dyn_clothB_03
  bone2Name=dyn_clothB_02
  swing1LimitDegrees=45
  swing1Motion=Free
  swing2LimitDegrees=45
  swing2Motion=Free
  twistLimitDegrees=10
 }
 {
  bone1Name=dyn_clothB_04
  bone2Name=dyn_clothB_03
  swing1LimitDegrees=45
  swing1Motion=Free
  swing2LimitDegrees=45
  swing2Motion=Free
  twistLimitDegrees=10
 }
 {
  bone1Name=dyn_clothB_02
  bone2Name=dyn_clothB_01
  swing1LimitDegrees=45
  swing1Motion=Free
  swing2LimitDegrees=45
  swing2Motion=Free
  twistLimitDegrees=10
 }
 {
  bone1Name=dyn_clothA_02
  bone2Name=dyn_clothA_01
  swing1LimitDegrees=45
  swing1Motion=Free
  swing2LimitDegrees=45
  swing2Motion=Free
  twistLimitDegrees=10
 }
 {
  bone1Name=Melee_TwoHanded_Handle
  bone2Name=b_AR_Root
  swing1LimitDegrees=45
  swing1Motion=Free
  swing2LimitDegrees=45
  swing2Motion=Free
  twistLimitDegrees=45
  twistMotion=Free
 }
 {
  bone1Name=dyn_base
  bone2Name=Melee_TwoHanded_Handle
  swing1LimitDegrees=45
  swing1Motion=Free
  swing2LimitDegrees=45
  swing2Motion=Free
  twistLimitDegrees=45
  twistMotion=Free
 }
 {
  bone1Name=dyn_clothA_01
  bone2Name=dyn_base
  swing1LimitDegrees=45
  swing1Motion=Free
  swing2LimitDegrees=45
  swing2Motion=Free
  twistLimitDegrees=10
 }
 {
  bone1Name=dyn_clothB_01
  bone2Name=dyn_base
  swing1LimitDegrees=45
  swing1Motion=Free
  swing2LimitDegrees=45
  swing2Motion=Free
  twistLimitDegrees=10
 }
 {bone1Name=dyn_cable_01, bone2Name=dyn_base, swing1LimitDegrees=20, swing2LimitDegrees=2, twistLimitDegrees=45, twistMotion=Locked}
 {bone1Name=dyn_cable_02, bone2Name=dyn_cable_01, swing1LimitDegrees=20, swing2LimitDegrees=2, twistLimitDegrees=45, twistMotion=Locked}
 {bone1Name=dyn_cable_03, bone2Name=dyn_cable_02, swing1LimitDegrees=20, swing2LimitDegrees=2, twistLimitDegrees=45, twistMotion=Locked}
 {bone1Name=dyn_cable_04, bone2Name=dyn_cable_03, swing1LimitDegrees=20, swing2LimitDegrees=2, twistLimitDegrees=45, twistMotion=Locked}
]
