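# Post-process style Anim Blueprint for the RobotRed pickaxe skeleton that adds secondary physics motion to the melee weapon.
# The input pose runs through a RigidBody node with its own physics asset, simulated in b_AR_Root space with scaled component motion and overridden gravity.
# Sine-driven piston and rumble bone offsets and a Perlin noise calculation are present but unwired, so only the physics affects the output pose.
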AnimBlueprint /BRCosmetics/Weapons/FORT_Melee/Pickaxe_RobotRed/Meshes/Pickaxe_RobotRed_Skeleton_AnimBlueprint
target_skeleton=&/Pickaxe_RobotRed_Skeleton
graphs [
 AnimGraph {nodes [
   "Transform (Modify) Bone - Bone: dyn_piston"!(ComponentPose, Alpha=1, Translation=Get pistonRun() -> (pistonRun:Vector)) -> (Pose) {translationMode=BMM_Additive, alphaBoolBlend {blendOption=Linear}, boneToModify {boneName=dyn_piston}}
   "Transform (Modify) Bone - Bone: dyn_rumble"!(ComponentPose, Alpha=1, Translation=Get pistonJitter() -> (pistonJitter:Vector)) -> (Pose) {
    translationMode=BMM_Additive
    translationSpace=BCS_ParentBoneSpace
    alphaBoolBlend {blendOption=Linear}
    alphaScaleBias {scale=0.2}
    alphaScaleBiasClamp {bInterpResult=true, interpSpeedDecreasing=1, interpSpeedIncreasing=1}
    boneToModify {boneName=dyn_rumble}
   }
   AnimGraphNode_SubInput_0 = Input Pose()
   n1 = RigidBody(ComponentPose=Local To Component(LocalPose=@AnimGraphNode_SubInput_0)) {
    bOverrideWorldGravity=true
    overridePhysicsAsset=&/Pickaxe_RobotRed_Physics
    simulationSpace=BaseBoneSpace
    alphaBoolBlend {blendOption=Linear}
    baseBoneRef {boneName=b_AR_Root}
    componentAppliedLinearAccClamp=<1000 1000 1000>
    componentLinearAccScale=<0.5 0.5 0.5>
    componentLinearVelScale=<0.4 0.4 0.4>
    overrideWorldGravity {z=-980}
   }
   Output Pose(Result=Component To Local(ComponentPose=@n1)) {layerGroup=None}
  ]}
 EventGraph {nodes [
   Event BlueprintInitializeAnimation()
   > Set randomStream(randomStream:Random Stream Structure=MakeRandomStream!(InitialSeed:int=0) -> (ReturnValue:Random Stream Structure)) -> (Output_Get:Random Stream Structure)
   n1 = Get randomStream() -> (randomStream:Random Stream Structure)
   BreakVector(InVec=Make Perlin Noise Vector and Remap!(
     X:float=RandomFloatInRangeFromStream!(Stream:"Random Stream Structure (by ref)"=@n1, Min:float=0, Max:float=10) -> (ReturnValue:float)
     Y:float=RandomFloatInRangeFromStream(Stream=@n1, Min=-1, Max=1)
     Z:float=RandomFloatInRangeFromStream(Stream=@n1, Min=-1, Max=1)
     RangeOutMinX:float=-1.f
     RangeOutMaxX:float=1.f
     RangeOutMinY:float=-0.5
     RangeOutMaxY:float=0.5
     RangeOutMinZ:float=-2
     RangeOutMaxZ:float=2
    ) -> (ReturnValue:Vector))
   Event BlueprintUpdateAnimation()
   n2 = GetGameTimeInSeconds()
   n3 = float * float(A=1, B=333)
   n4 = float * float(A="Sin (Radians)"(A=float * float(A=@n3, B=@n2)), B=1.5)
   Set pistonJitter(pistonJitter:Vector=MakeVector(X=@n4, Y=0, Z=@n4)) -> (Output_Get:Vector)
   n5 = "Sin (Degrees)"!(A:double=float * float(A=@n2, B=500)) -> (ReturnValue:double)
   Set pistonRun(pistonRun:Vector=MakeVector(Y=float * float(A=@n5, B=1.75))) -> (Output_Get:Vector)
  ]}
]
variables [
 pistonJitter {type=Vector}
 pistonRun {type=Vector}
 randomStream {type=RandomStream, default {initialSeed=0, seed=0}}
]
